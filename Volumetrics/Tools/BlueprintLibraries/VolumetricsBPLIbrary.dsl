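; Blueprint function library of shared helpers for the Volumetrics tools, covering materials, render targets, textures and actor lookup.
; CreateandAssign functions check by-ref MIDs and render targets against parent, or size and format, to create and assign replacements.
; PseudoVolumeDimensions maps an enum to 2D atlas resolution, tile grid and volume size for flipbook-style pseudo-volume textures.
; Other helpers read Texture2D or render target sizes, compare render target sizes and fetch the first actor of a class.
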(fn CreateandAssignMID
  (params (MID :type Material%20Instance%20Dynamic%20Object%20Reference%20%28by%20ref%29) (Parent :type Material%20Interface%20Object%20Reference))
  (Utilities|IsValid MID
    (:Is%20Valid
      (if (Utilities|Equal(Object) (Variables|/Script/Engine.MaterialInstance|GetParent MID) Parent)
        _
        (else
          _)))
    (:Is%20Not%20Valid
      _))
  (bind _returnvalue (Rendering|Material|CreateDynamicMaterialInstance (Variables|Get__WorldContext) Parent "None" "None" :!class /Script/Engine.KismetMaterialLibrary))
  (Variables|SetBy-RefVar MID _returnvalue))

(fn CreateandAssignRT
  (params (RT :type Texture%20Render%20Target%202D%20Object%20Reference%20%28by%20ref%29) (Size :type Int%20Point%20Structure) (Format :type ETextureRenderTargetFormat%20Enum :default "RTF_R8"))
  (bind (_x _y) (Utilities|Struct|BreakIntPoint Size))
  (bind (_x_2 _y_2) (Utilities|Struct|BreakIntPoint Size))
  (Utilities|IsValid RT
    (:Is%20Valid
      (if (and (and (== (Variables|/Script/Engine.TextureRenderTarget2D|GetSizeX RT) _x) (== (Variables|/Script/Engine.TextureRenderTarget2D|GetSizeY RT) _y)) (Utilities|Enum|Equal(Enum) (Variables|/Script/Engine.TextureRenderTarget2D|GetRenderTargetFormat RT) Format))
        _
        (else
          _)))
    (:Is%20Not%20Valid
      _))
  (bind _returnvalue (Rendering|CreateRenderTarget2D (Variables|Get__WorldContext) _x_2 _y_2 Format))
  (Variables|SetBy-RefVar RT _returnvalue))

(fn GetTextureSize
  (params (Texture :type Texture%20Object%20Reference))
  (bind _astexture_2d (Utilities|Casting|CastToTexture2D Texture)
    (:then
      (LocalVariables|SetOut%20Size (Utilities|Struct|MakeIntPoint (Rendering|Texture|GetSizeX _astexture_2d) (Rendering|Texture|GetSizeY _astexture_2d))))
    (:CastFailed
      (bind _astexture_render_target_2d (Utilities|Casting|CastToTextureRenderTarget2D Texture)
        (:then
          (LocalVariables|SetOut%20Size (Utilities|Struct|MakeIntPoint (Variables|/Script/Engine.TextureRenderTarget2D|GetSizeX _astexture_render_target_2d) (Variables|/Script/Engine.TextureRenderTarget2D|GetSizeY _astexture_render_target_2d))))
        (:CastFailed
          (Development|PrintString 0 "Warning: Input Texture failed ot cast to either Texture2D or TextureRenderTarget2D.")))))
  (return (LocalVariables|GetOut%20Size)))

(fn GetFirstActorOfClass
  (params (WorldContent :type Object%20Reference) (Class :type Actor%20Class%20Reference))
  (bind _outactors (Actor|GetAllActorsOfClass WorldContent Class))
  (if (> (Utilities|Array|Length _outactors) 0)
    (Utilities|IsValid (Utilities|Array|Get(acopy) _outactors)
      (:Is%20Valid
        (LocalVariables|SetReturn%20Actor (Utilities|Array|Get(acopy) _outactors)))
      (:Is%20Not%20Valid
        (LocalVariables|SetReturn%20Actor 0)))
    (return (LocalVariables|GetReturn%20Actor))
    (else
      _)))

(fn PseudoVolumeDimensions
  (params (Dimensions :type EVolumeTextureDimensions%20Enum :default "NewEnumerator0"))
  (return (Utilities|Vector2DToIntPoint (Utilities|Select :Index Dimensions :NewEnumerator0 "(X=1024.000000,Y=32.000000)" :NewEnumerator1 "(X=512.000000,Y=512.000000)" :NewEnumerator2 "(X=1000.000000,Y=1000.000000)" :NewEnumerator3 "(X=2048.000000,Y=1024.000000)" :NewEnumerator4 "(X=2744.000000,Y=2744.000000)" :NewEnumerator5 "(X=4096.000000,Y=4096.000000)" :NewEnumerator6 "(X=5832.000000,Y=5832.000000)" :NewEnumerator7 "(X=8000.000000,Y=8000.000000)")) (Utilities|Vector2DToIntPoint (Utilities|Select :Index Dimensions :NewEnumerator0 "(X=32.000000,Y=1.000000)" :NewEnumerator1 "(X=8.000000,Y=8.000000)" :NewEnumerator2 "(X=10.000000,Y=10.000000)" :NewEnumerator3 "(X=16.000000,Y=8.000000)" :NewEnumerator4 "(X=14.000000,Y=14.000000)" :NewEnumerator5 "(X=16.000000,Y=16.000000)" :NewEnumerator6 "(X=18.000000,Y=18.000000)" :NewEnumerator7 "(X=20.000000,Y=20.000000)")) (Utilities|Select :Index Dimensions :NewEnumerator0 32 :NewEnumerator1 64 :NewEnumerator2 100 :NewEnumerator3 128 :NewEnumerator4 192 :NewEnumerator5 256 :NewEnumerator6 324 :NewEnumerator7 400)))

(fn RTSizeCompare
  (params (A :type Texture%20Render%20Target%202D%20Object%20Reference%20%28by%20ref%29) (B :type Texture%20Render%20Target%202D%20Object%20Reference%20%28by%20ref%29))
  (return (and (and (Utilities|IsValid A :!class /Script/Engine.KismetSystemLibrary) (Utilities|IsValid B :!class /Script/Engine.KismetSystemLibrary)) (Utilities|IntPointCompare (Utilities|Struct|MakeIntPoint (Variables|/Script/Engine.TextureRenderTarget2D|GetSizeX A) (Variables|/Script/Engine.TextureRenderTarget2D|GetSizeY A)) (Utilities|Struct|MakeIntPoint (Variables|/Script/Engine.TextureRenderTarget2D|GetSizeX B) (Variables|/Script/Engine.TextureRenderTarget2D|GetSizeY B))))))
